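# Sound cue for a Jones dialogue line in the Battle Pass S20 Week 8 voice-over set, named as line 010 of group 02_03.
# The cue has a null root and no sound nodes, so the file shows no wave or playback logic.
# It carries only routing settings: the shared S20 quest attenuation, the narrator VO sound class, a raised volume multiplier and a very large max distance.
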SoundCue /BattlePassS20/Audio/VO/Cues/Week_8/02/JONES_2022_REGS_Week08_02_03_010_Dialogue_Cue
root=null
nodes []
properties {
 attenuationSettings=/BattlePassS20/Audio/VO/S20_Quest_Att
 maxDistance=2097152
 soundClassObject=/Game/Sounds/FortSoundClassesAndModes/Fort_VO_Narrator
 volumeMultiplier=2
}
